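; ToyFox outro gameplay ability that plays the Megazord rez-out and exits the giant.
; Activation fires a start cue, plays the montage, whose notify triggers the rez-out cue, and hides the HUD via gameplay effect.
; The authority switches secondary pilots to third-person view or exits them per hotfix, then calls HandleOutroFinished after a hotfixable delay.
; On end, non-authority instances activate the exit-giant ability and the HUD effect is removed.
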(event Ability|EventActivateAbility
  ;[0.003 0.125 0 0.5] ActivateAbility: Play the rez-out animation and exit the giant.
  (|CallParentFunction)
  (Ability|ExecuteGameplayCueOnOwner (Variables|GameplayCue|GetGC_OutroStart))
  (bind (_asynctaskproxy _notifytag) (Ability|Tasks|PlayMontageAndWaitNotify "RezOutMontage" (Variables|Settings|GetRezOutMontage))
    (:then
      (Variables|Transient|SetAvatarPlayerController (Fortnite|GetFortPlayerControllerfromActor (CallFunction|GetAvatarasFortPlayerPawn)))
      (bind _returnvalue (Ability|ApplyGameplayEffectToOwner (Variables|Settings|GetHideHUDGameplayEffectClass)))
      (Variables|Transient|SetHideHUDGameplayEffectHandle _returnvalue)
      (Ability|HasAuthority
        (:True
          (Utilities|FlowControl|Sequence
            (:then_0
              (bind _gamestatecomponent (Utilities|GetGameStateComponent)
                (:Valid
                  (if (> (ScalableFloat|GetCachedScalableFloatValue (Variables|Hotfix|GetHotfix_2ndPilot3PView)) 0)
                    (Class|BToyFoxGameStateComponent|SetSecondaryPilot3PView _gamestatecomponent true) ; Disable will auto trigger for exiting each pilot
                    (else
                      (Class|BToyFoxGameStateComponent|ExitSecondaryPilots _gamestatecomponent))))))
            (:then_1
              (Utilities|FlowControl|Delay (ScalableFloat|GetValueatLevel (Variables|Hotfix|GetHotfix_OutroDuration)))
              (bind _gamestatecomponent_2 (Utilities|GetGameStateComponent)
                (:Valid
                  (Class|BToyFoxGameStateComponent|HandleOutroFinished _gamestatecomponent_2))))))))
    (:OnNotifyBegin
      (if (== _notifytag "(TagName=\"AnimNotify.ToyFox.Ability.Outro.RezOut\")")
        (Ability|ExecuteGameplayCueOnOwner (Utilities|Struct|BreakGameplayCueTag (Variables|GameplayCue|GetRezOutGameplayCueTag)))))))

(event Ability|EventOnEndAbility
  (params (bWasCancelled :type Boolean))
  ;[0.09896 0.05702 0 0.5] EndAbility: Activate the ability to exit megazord.
  (Utilities|FlowControl|Sequence
    (:then_0
      (Ability|HasAuthority
        (:True
          _)
        (:False
          (Abilities|TryActivateAbilitybyClass (Ability|GetAbilitySystemComponentfromActorInfo) (Variables|Settings|GetExitGiantGameplayAbility) true)))
      (Ability|RemoveGameplayEffectFromOwnerWithHandle (Variables|Transient|GetHideHUDGameplayEffectHandle)))
    (:then_1
      (|CallParentFunction bWasCancelled))))

(fn CanActivateAbility
  (params (ActorInfo :type Gameplay%20Ability%20Actor%20Info%20Structure) (Handle :type Gameplay%20Ability%20Spec%20Handle%20Structure))
  (return 0 true))
